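# Looping ambient water cue for the outer white water of the CubeIsland whirlpool, limited to one voice.
# A quality-level root selects either a three-loop mix of swirling water and waterfalls, dominated by a large waterfall-base loop, or one of three pitch-modulated single waterfall-roar loops.
# Each branch overrides attenuation with a custom falloff curve, occlusion, listener focus and wide stereo spread, with the mix fading out sooner than the roars.
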SoundCue /BlastBerryMapContent/Game/Athena/Sounds/Events/CubeIsland/Spot_WhirlPool_Outer_White_Water_Loop_Cue
root=QualityLevel_0
nodes [
 {
  id=Attenuation_5
  children [Mixer_0]
  properties {
   bOverrideAttenuation=true
   attenuationOverrides {
    bEnableFocusInterpolation=true
    bEnableListenerFocus=true
    bEnableOcclusion=true
    distanceAlgorithm=Custom
    falloffDistance=7500
    lPFFrequencyAtMax=1000
    lPFRadiusMax=4000
    lPFRadiusMin=1000
    nonFocusVolumeAttenuation=0.8
    nonSpatializedRadiusStart=2000
    occlusionInterpolationTime=5
    occlusionLowPassFilterFrequency=1000
    occlusionTraceChannel=ECC_GameTraceChannel11
    occlusionVolumeAttenuation=0.5
    stereoSpread=10000
    attenuationShapeExtents {x=0}
    customAttenuationCurve {editorCurveData {keys [
       <at 0 1>
       {arriveTangent=0.028, interpMode=RCIM_Cubic, leaveTangent=0.028, tangentMode=RCTM_User, time=0.1295, value=0.99}
       {arriveTangent=-1.8481, interpMode=RCIM_Cubic, leaveTangent=-1.8481, tangentMode=RCTM_User, time=0.22, value=0.5573}
       <at 0.75 0>
      ]}}
   }
  }
 }
 {
  id=Attenuation_6
  children [Modulator_0]
  properties {
   bOverrideAttenuation=true
   attenuationOverrides {
    bEnableFocusInterpolation=true
    bEnableListenerFocus=true
    bEnableOcclusion=true
    distanceAlgorithm=Custom
    falloffDistance=7500
    lPFFrequencyAtMax=1000
    lPFRadiusMax=4000
    lPFRadiusMin=1000
    nonFocusVolumeAttenuation=0.8
    nonSpatializedRadiusStart=5000
    occlusionInterpolationTime=5
    occlusionLowPassFilterFrequency=1000
    occlusionTraceChannel=ECC_GameTraceChannel11
    occlusionVolumeAttenuation=0.5
    stereoSpread=10000
    attenuationShapeExtents {x=0}
    customAttenuationCurve {editorCurveData {keys [
       <at 0 1>
       {arriveTangent=0.028, interpMode=RCIM_Cubic, leaveTangent=0.028, tangentMode=RCTM_User, time=0.22, value=1}
       {arriveTangent=-0.2554, interpMode=RCIM_Cubic, leaveTangent=-0.2554, tangentMode=RCTM_User, time=0.4, value=0.97}
       <at 1 0>
      ]}}
   }
  }
 }
 {
  id=Attenuation_7
  children [Modulator_1]
  properties {
   bOverrideAttenuation=true
   attenuationOverrides {
    bEnableFocusInterpolation=true
    bEnableListenerFocus=true
    bEnableOcclusion=true
    distanceAlgorithm=Custom
    falloffDistance=7500
    lPFFrequencyAtMax=1000
    lPFRadiusMax=4000
    lPFRadiusMin=1000
    nonFocusVolumeAttenuation=0.8
    nonSpatializedRadiusStart=5000
    occlusionInterpolationTime=5
    occlusionLowPassFilterFrequency=1000
    occlusionTraceChannel=ECC_GameTraceChannel11
    occlusionVolumeAttenuation=0.5
    stereoSpread=10000
    attenuationShapeExtents {x=0}
    customAttenuationCurve {editorCurveData {keys [
       <at 0 1>
       {arriveTangent=0.028, interpMode=RCIM_Cubic, leaveTangent=0.028, tangentMode=RCTM_User, time=0.22, value=1}
       {arriveTangent=-0.2554, interpMode=RCIM_Cubic, leaveTangent=-0.2554, tangentMode=RCTM_User, time=0.4, value=0.97}
       <at 1 0>
      ]}}
   }
  }
 }
 {
  id=Attenuation_8
  children [Modulator_3]
  properties {
   bOverrideAttenuation=true
   attenuationOverrides {
    bEnableFocusInterpolation=true
    bEnableListenerFocus=true
    bEnableOcclusion=true
    distanceAlgorithm=Custom
    falloffDistance=7500
    lPFFrequencyAtMax=1000
    lPFRadiusMax=4000
    lPFRadiusMin=1000
    nonFocusVolumeAttenuation=0.8
    nonSpatializedRadiusStart=5000
    occlusionInterpolationTime=5
    occlusionLowPassFilterFrequency=1000
    occlusionTraceChannel=ECC_GameTraceChannel11
    occlusionVolumeAttenuation=0.5
    stereoSpread=10000
    attenuationShapeExtents {x=0}
    customAttenuationCurve {editorCurveData {keys [
       <at 0 1>
       {arriveTangent=0.028, interpMode=RCIM_Cubic, leaveTangent=0.028, tangentMode=RCTM_User, time=0.22, value=1}
       {arriveTangent=-0.2554, interpMode=RCIM_Cubic, leaveTangent=-0.2554, tangentMode=RCTM_User, time=0.4, value=0.97}
       <at 1 0>
      ]}}
   }
  }
 }
 {id=Mixer_0, children [WavePlayer_2, WavePlayer_3, WavePlayer_10], properties {inputVolume [1.5,7.5,1.5]}}
 {id=Mixer_2, children [Attenuation_5, null], properties {inputVolume [1,1]}}
 {id=Modulator_0, children [WavePlayer_6], properties {pitchMax=1.3, pitchMin=1.2}}
 {id=Modulator_1, children [WavePlayer_7], properties {pitchMax=1.1, pitchMin=0.9}}
 {id=Modulator_3, children [WavePlayer_9], properties {pitchMax=1.1, pitchMin=0.9}}
 {id=QualityLevel_0, children [Mixer_2, Attenuation_6, Attenuation_7, Attenuation_8]}
 {id=WavePlayer_10, properties {bLooping=true, soundWaveAssetPtr=/Game/Sounds/Fort_Ambient_Looping/Water/waterfall_new_ST_01}}
 {
  id=WavePlayer_2
  properties {bLooping=true, soundWaveAssetPtr=/BlastberryMapContent/Game/Athena/Sounds/Events/CubeIsland/Spot_WhirlPool_Outer_White_Water_Swirling_Loop}
 }
 {id=WavePlayer_3, properties {bLooping=true, soundWaveAssetPtr=/Game/Athena/Sounds/Water/Ambient/Amb_Water_Waterfall_Large_Bot_Loop_01}}
 {
  id=WavePlayer_6
  properties {bLooping=true, soundWaveAssetPtr=/Game/Sounds/Artemis/POI/RealityFalls/Spots/Waterfalls/Spot_RealityFalls_Waterfall_Roar_Close}
 }
 {
  id=WavePlayer_7
  properties {bLooping=true, soundWaveAssetPtr=&Spot_RealityFalls_Waterfall_Roar_Close}
 }
 {
  id=WavePlayer_9
  properties {bLooping=true, soundWaveAssetPtr=&Spot_RealityFalls_Waterfall_Roar_Close}
 }
]
properties {
 bOverrideConcurrency=true
 maxDistance=7500
 pitchMultiplier=0.9
 soundClassObject=/Game/Sounds/FortSoundClassesAndModes/Fort_Water_Ambient
 volumeMultiplier=0.8
 attenuationOverrides {
  bAttenuateWithLPF=true
  bEnableListenerFocus=true
  falloffDistance=6000
  lPFFrequencyAtMax=500
  nonSpatializedRadiusStart=700
  stereoSpread=1000
  attenuationShapeExtents {x=925}
 }
 concurrencyOverrides {maxCount=1, voiceStealReleaseTime=5}
}
